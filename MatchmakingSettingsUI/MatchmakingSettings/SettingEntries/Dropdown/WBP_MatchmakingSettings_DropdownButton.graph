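# Dropdown button for matchmaking settings entries, a child of the generic UIKit button with no widgets of its own.
# PreConstruct calls SetMobileStyle, which keeps the current font but picks a larger size when compact UI is in use.
# Anim_ContextOpen animates the AnimHack_Arrow proxy variable from 0 to 1; its setter turns that into a half-turn rotation of the block icon, avoiding direct animation of the inherited widget.
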WidgetBlueprint /MatchmakingSettingsUI/MatchmakingSettings/SettingEntries/Dropdown/WBP_MatchmakingSettings_DropdownButton
animations [Anim_ContextOpen {
  binding_count=1
  length_seconds=0.2
  object=&self:Anim_ContextOpen
  sequence {bindings [WBP_MatchmakingSettings_DropdownButton_C_23 {
     id=#1
     possessed_class=WBP_MatchmakingSettings_DropdownButton_C
     tracks [{
       type=MovieSceneDoubleTrack
       display_name=Anim Hack Arrow
       sections [{type=MovieSceneDoubleSection, end_frame=4, start_frame=0, channels [None {type=MovieSceneScriptingDoubleChannel, keys [[0,0],[4,1]]}]}]
      }]
    }]}
 }]
graphs [
 EventGraph {nodes [
   n1 = Event PreConstruct()
   > Parent: PreConstruct(IsDesignTime=@n1.IsDesignTime)
   > SetMobileStyle() {is=title K2Node_CallFunction}
   Event Construct()
   > Parent: Construct()
  ]}
 SetAnimHack_Arrow {nodes [
   n1 = SetAnimHack_Arrow!() -> (then, NewParam:double=~) {is=title K2Node_FunctionEntry}
   > n2 = Get!(execute:Exec) -> (then:Exec, else:Exec, Block_Icon:WBP UIKit Block Image Object Reference)
   > SetRenderTransformAngle!(execute, self:Widget Object Reference=@n2, Angle:float=float * float(A=@n1, B=180) {is=title K2Node_PromotableOperator}) -> (then)
  ]}
 SetMobileStyle {nodes [
   SetMobileStyle() {is=title K2Node_FunctionEntry}
   n1 = Get!() -> (
    Font_FontObject:Object Reference
    Font_FontMaterial:Object Reference
    Font_OutlineSettings:Font Outline Settings Structure
    Font_TypefaceFontName:Name
    Font_Size:float=~
    Font_LetterSpacing:int
    Font_SkewAmount:float=~
    Font_bForceMonospaced:bool=~
    Font_bMaterialIsStencil:bool=~
    Font_MonospacedWidth:float=~
   )
   > SetFont!(
    execute
    self:WBP UIKit Button Generic Base Object Reference
    In Font_FontObject:Object Reference=@n1
    In Font_FontMaterial:Object Reference=@n1.Font_FontMaterial
    In Font_OutlineSettings:Font Outline Settings Structure=@n1.Font_OutlineSettings
    In Font_TypefaceFontName:Name=@n1.Font_TypefaceFontName
    In Font_Size:float=Select(Option 0:float=21, Option 1:float=24, Index=IsUsingCompactUI()) -> (ReturnValue:float)
    In Font_LetterSpacing:int=@n1.Font_LetterSpacing
    In Font_SkewAmount:float=@n1.Font_SkewAmount
    In Font_bForceMonospaced:bool=@n1.Font_bForceMonospaced
    In Font_bMaterialIsStencil:bool=@n1.Font_bMaterialIsStencil
    In Font_MonospacedWidth:float=@n1.Font_MonospacedWidth
   ) -> (then)
  ]}
]
variables [AnimHack_Arrow {type=number}]
widgets {
 info {
  inheritedWidgetCount=0
  parentClass=/FortUILibrary/UIKit/Buttons/WBP_UIKit_Button_Generic.WBP_UIKit_Button_Generic_C
  rootWidgetClass=null
  widgetCount=0
 }
 widgets []
}
